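; RandomColor_Prop is a Fortnite prop Blueprint that varies its color swatch, add-on mesh and material flipbook frame.
; The construction script can pick each index deterministically from world position, using averaged absolute coordinates over a transition distance, modulo option count.
; Validation applies palette swatches as building texture data, then loads the mesh (blocking in editor, async at runtime) with optional socket transform and material override.
; The flipbook index drives a material scalar parameter.
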(event EventBeginPlay
  (|CallParentFunction)
  (Variables|Default|SetcurrentPalette 0) ; Release Current Palette object
)

(event Custom|LoadMesh
  (bind _meshindex (Variables|RandomMesh|GetMeshIndex))
  (bind _meshlist (Variables|RandomMesh|GetMeshList))
  (bind _output (Utilities|Array|Get(acopy) _meshlist _meshindex))
  (Utilities|IsRuntime
    (:EditorTime
      (bind _returnvalue (Utilities|LoadAssetBlocking _output))
      (bind _asstatic_mesh (Utilities|Casting|CastToStaticMesh _returnvalue))
      (Variables|Default|SetCurrentStaticMeshLoaded _asstatic_mesh))
    (:Runtime
      (bind _object (Utilities|AsyncLoadAsset _output)
        (:Completed
          (bind _asstatic_mesh_2 (Utilities|Casting|CastToStaticMesh _object))
          (Variables|Default|SetCurrentStaticMeshLoaded _asstatic_mesh_2)))))
  (Utilities|IsValid (Variables|Default|GetCurrentLoadedMeshComponent)
    (:Is%20Valid
      _)
    (:Is%20Not%20Valid
      (if (Variables|RandomMesh|GetUseSocketTransform)
        (Default|SetMeshTransform)
        (else
          (Variables|Default|SetAddOnMesh_Transform)))
      (bind _returnvalue_2 (AddComponent|NODE_AddStaticMeshComponent-0 :RelativeTransform (Variables|Default|GetAddOnMesh_Transform)))
      (bind _output_get (Variables|Default|SetCurrentLoadedMeshComponent _returnvalue_2))))
  (Components|StaticMesh|SetStaticMesh (Variables|Default|GetCurrentLoadedMeshComponent) (Variables|Default|GetCurrentStaticMeshLoaded))
  (RandomMesh|ValidateMeshMaterials)
  (Class|CreativeAssetLoadingTrackingComponent|NotifyAssetsLoaded (Variables|Default|GetCreativeAssetLoadingTracking))
  (if (Variables|Default|GetHideParentComponent)
    (Rendering|SetVisibility _output_get false false)))

(fn BuildExposedAssetProviderVersion(Implementation)
  (params (Version :type Fort%20Exposed%20Asset%20Provider%20Version%20Structure%20%28by%20ref%29))
  (|CallParentFunction Version)
  (ExposedAssetProvider|AddVersion Version 1)
  (return))

(fn CollectAssetReferencestoAllow(Implementation)
  (params (Collector :type Fort%20Exposed%20Asset%20Reference%20Collector%20Structure%20%28by%20ref%29))
  (|CallParentFunction Collector)
  (ExposedAssetProvider|CollectObjectPropertyAssetReferences Collector 0 (Variables|RandomMesh|GetMeshList))
  (return))

(fn CollectClassReferencestoAllow(Implementation)
  (params (Collector :type Fort%20Exposed%20Class%20Reference%20Collector%20Structure%20%28by%20ref%29))
  (|CallParentFunction Collector)
  (ExposedAssetProvider|CollectObjectPropertyClassReferences Collector 0 (Variables|RandomSwatch|GetSwatchDataAsset))
  (return))

(fn CollectExposedAssetProviderReferences(Implementation)
  (params (Collector :type Fort%20Exposed%20Asset%20Reference%20Collector%20Structure%20%28by%20ref%29))
  (|CallParentFunction Collector)
  (ExposedAssetProvider|CollectObjectPropertyAssetReferences Collector 0 (Variables|RandomSwatch|GetSwatchDataAsset))
  (return))

(fn FlipBookCanRun)

(fn GetSwatchArray
  (bind _swatchdataasset (Variables|RandomSwatch|GetSwatchDataAsset))
  (if (Utilities|IsValidClass _swatchdataasset)
    (Utilities|IsValid (Variables|Default|GetcurrentPalette)
      (:Is%20Valid
        _)
      (:Is%20Not%20Valid
        (bind _returnvalue (Game|ConstructObjectfromClass _swatchdataasset self))
        (Variables|Default|SetcurrentPalette _returnvalue)))
    (return (Variables|/Game/Environments/Apollo/Blueprints/RandomColor_Prop/RandomColor_Prop_Palette.RandomColor_Prop_Palette_C|GetSwatch_Palette (Variables|Default|GetcurrentPalette)) (Utilities|Array|Length (Variables|/Game/Environments/Apollo/Blueprints/RandomColor_Prop/RandomColor_Prop_Palette.RandomColor_Prop_Palette_C|GetSwatch_Palette (Variables|Default|GetcurrentPalette))))
    (else
      (Variables|Default|SetcurrentPalette 0))))

(fn GetSwatchArrayConst
  (bind _swatchdataasset (Variables|RandomSwatch|GetSwatchDataAsset))
  (if (Utilities|IsValidClass _swatchdataasset)
    (bind _returnvalue (Game|ConstructObjectfromClass _swatchdataasset))
    (bind _output_get (LocalVariables|SetLocalPalette _returnvalue))
    (return (Variables|/Game/Environments/Apollo/Blueprints/RandomColor_Prop/RandomColor_Prop_Palette.RandomColor_Prop_Palette_C|GetSwatch_Palette _output_get) (Utilities|Array|Length (Variables|/Game/Environments/Apollo/Blueprints/RandomColor_Prop/RandomColor_Prop_Palette.RandomColor_Prop_Palette_C|GetSwatch_Palette _output_get)))
    (else
      (return 0 0))))

(fn RandomFlipBookByPosition
  (bind _flipbookrowcolumncount (Variables|FlipBook|GetFlipBookRowColumnCount))
  (bind (_x _y) (Math|Vector2D|BreakVector2D _flipbookrowcolumncount))
  (bind _returnvalue (* _x _y))
  (bind _returnvalue_2 (Math|Float|Floor _returnvalue))
  (bind _flipbooktransitiondistance (Variables|FlipBook|GetFlipBookTransitionDistance))
  (if (> _returnvalue_2 0)
    (if (Math|Float|NearlyEqual(Float) _flipbooktransitiondistance)
      (Variables|FlipBook|SetFlipBookIndex 0)
      (else
        (Variables|FlipBook|SetFlipBookIndex (+ (Math|Integer|%25%28Integer%29 (Math|Float|Round (/ (/ (Math|Float|float+float (Math|Float|Absolute(Float) (.x (Transformation|GetActorLocation))) (Math|Float|Absolute(Float) (.y (Transformation|GetActorLocation))) (Math|Float|Absolute(Float) (.z (Transformation|GetActorLocation)))) 3.0) _flipbooktransitiondistance)) _returnvalue_2) 1))))))

(fn RandomMeshByPosition
  (bind _meshlist (Variables|RandomMesh|GetMeshList))
  (bind _returnvalue (Utilities|Array|Length _meshlist))
  (bind _meshtransitiondistance (Variables|RandomMesh|GetMeshTransitionDistance))
  (if (> _returnvalue 0)
    (if (Math|Float|NearlyEqual(Float) _meshtransitiondistance)
      (Variables|RandomMesh|SetMeshIndex 0)
      (else
        (Variables|RandomMesh|SetMeshIndex (Math|Integer|%25%28Integer%29 (Math|Float|Round (/ (/ (Math|Float|float+float (Math|Float|Absolute(Float) (.x (Transformation|GetActorLocation))) (Math|Float|Absolute(Float) (.y (Transformation|GetActorLocation))) (Math|Float|Absolute(Float) (.z (Transformation|GetActorLocation)))) 3.0) _meshtransitiondistance)) _returnvalue))))))

(fn RandomSwatchByPosition
  (bind _swatchtransitiondistance (Variables|RandomSwatch|GetSwatchTransitionDistance))
  (bind (_color_swatches _swatch_count) (SwatchList|GetSwatchArray))
  (if (> _swatch_count 0)
    (if (Math|Float|NearlyEqual(Float) _swatchtransitiondistance)
      (Variables|RandomSwatch|SetSwatchIndex 0)
      (else
        (Variables|RandomSwatch|SetSwatchIndex (Math|Integer|%25%28Integer%29 (Math|Float|Round (/ (/ (Math|Float|float+float (Math|Float|Absolute(Float) (.x (Transformation|GetActorLocation))) (Math|Float|Absolute(Float) (.y (Transformation|GetActorLocation))) (Math|Float|Absolute(Float) (.z (Transformation|GetActorLocation)))) 3.0) _swatchtransitiondistance)) _swatch_count))))))

(fn RunRandomization
  (Utilities|FlowControl|Sequence
    (:then_0
      (if (Variables|RandomMesh|GetRandomizeMeshByPosition)
        (RandomMesh|RandomMeshbyPosition)))
    (:then_1
      (if (Variables|RandomSwatch|GetRandomizeSwatchByPosition)
        (RandomSwatch|RandomSwatchbyPosition)))
    (:then_2
      (Utilities|CanExecuteFlipBook
        (:Valid
          (if (Variables|FlipBook|GetRandomizeFlipBookByPosition)
            (FlipBook|RandomFlipBookbyPosition)))))
    (:then_3
      (CallFunction|ValidateVisualsAll))))

(fn SetExactSwatchIndex
  (params (SwatchIndex :type Integer))
  (Variables|RandomSwatch|SetSwatchIndex SwatchIndex)
  (CallFunction|ValidateVisualsAll))

(fn SetMeshTransform
  (bind _staticmeshcomponent (Variables|Debug|GetStaticMeshComponent))
  (bind _returnvalue (Transformation|GetAllSocketNames _staticmeshcomponent))
  (if (> (Utilities|Array|Length _returnvalue) 0)
    (Variables|Default|SetAddOnMesh_Transform (Transformation|GetSocketTransform _staticmeshcomponent (Utilities|Array|Get(acopy) _returnvalue) "RTS_Actor"))
    (else
      (Variables|Default|SetAddOnMesh_Transform))))

(fn SetSwatchData
  (params (SwatchDataAsset :type Random%20Color%20Prop%20Palette%20Class%20Reference))
  (if (Utilities|IsValidClass SwatchDataAsset)
    (bind _returnvalue (Game|ConstructObjectfromClass SwatchDataAsset))
    (Utilities|IsValid _returnvalue
      (:Is%20Valid
        (Variables|Default|SetcurrentPalette _returnvalue)
        (Variables|RandomSwatch|SetSwatchDataAsset SwatchDataAsset)
        (Variables|RandomSwatch|SetSwatchIndex (Math|Integer|Clamp(Integer) (Variables|RandomSwatch|GetSwatchIndex) 0 (Utilities|Array|LastIndex (Variables|/Script/FortniteGame.FortRandomColorPropPalette|GetSwatch_Palette _returnvalue))))
        (CallFunction|ValidateVisualsAll)))))

(fn ShouldRandomize
  (return (and (< (Variables|RandomSwatch|GetSwatchIndex) 0) (< (Variables|RandomMesh|GetMeshIndex) 0))))

(fn ConstructionScript
  (|CallParentFunction)
  (if (not (Variables|Mesh|GetOverrideStaticMeshMobility))
    (for _array_element (Actor|GetComponentsByClass "/Script/Engine.StaticMeshComponent")
      (Transformation|SetMobility _array_element "Static")
      (Utilities|IsRuntime
        (:EditorTime
          _)
        (:Runtime
          (if (not (Default|ShouldRandomize))
            (CallFunction|ValidateVisualsAll))))
      (RandomSwatch|RunRandomization))))

(fn ValidateVisualsAll
  (Utilities|FlowControl|Sequence
    (:then_0
      (SwatchList|ValidateVisualsSwatches) ; Needs to be before mesh visual validation
    )
    (:then_1
      (MeshList|ValidateVisualsMeshes))
    (:then_2
      (Utilities|CanExecuteFlipBook
        (:Valid
          (FlipBook|ValidateVisualsFlipBook))))))

(fn ValidateMeshMaterials
  (bind _currentloadedmeshcomponent (Variables|Default|GetCurrentLoadedMeshComponent))
  (switch Utilities|FlowControl|Switch|SwitchonRandomColorProp_MaterialApply_Enum (Variables|RandomMesh|GetMeshMaterialOverride)
    (:NewEnumerator1
      ;[1 1 1] Copy the base mesh materials to the newly load mesh
      (Utilities|IsValid (Variables|Default|GetCurrentLoadedMeshComponent)
        (:Is%20Valid
          (for (_array_element _array_index) (Rendering|Material|GetMaterials (Variables|Debug|GetStaticMeshComponent))
            (if (< _array_index (Rendering|Material|GetNumMaterials _currentloadedmeshcomponent))
              (Rendering|Material|SetMaterial _currentloadedmeshcomponent _array_index _array_element))))))
    (:NewEnumerator2
      (bind _outtexturedataarray (Building|GetBuildingTextureData))
      (if (Utilities|Array|IsValidIndex _outtexturedataarray)
        (Utilities|CopyDataOverridestoMesh (Variables|Default|GetCurrentLoadedMeshComponent) (Utilities|Array|Get(acopy) _outtexturedataarray))))))

(fn ValidateVisuals_Meshes
  (Utilities|FlowControl|Sequence
    (:then_0
      (if (== (Utilities|Array|Length (Variables|RandomMesh|GetMeshList)) 0)
        (return)))
    (:then_1
      (if (>= (Variables|RandomMesh|GetMeshIndex) (Utilities|Array|Length (Variables|RandomMesh|GetMeshList)))
        (Variables|RandomMesh|SetMeshIndex (- (Utilities|Array|Length (Variables|RandomMesh|GetMeshList)) 1))
        (elif (< (Variables|RandomMesh|GetMeshIndex) 0)
          (Variables|RandomMesh|SetMeshIndex 0))))
    (:then_2
      (CallFunction|LoadMesh))))

(fn ValidateVisuals_Swatches
  (bind _swatchindex (Variables|RandomSwatch|GetSwatchIndex))
  (bind (_color_swatches _swatch_count) (SwatchList|GetSwatchArray))
  ;[1 1 1] Comment
  (if (> _swatch_count 0)
    ;[1 1 1] Comment
    (if (Utilities|Array|IsValidIndex _color_swatches (Variables|RandomSwatch|GetSwatchIndex))
      ;[1 1 1] Comment
      (CallFunction|SetBuildingTextureData 0 (Utilities|Array|Get(acopy) _color_swatches _swatchindex))
      (if (Variables|Default|GetApplySwatchToSecondDataOverride)
        (CallFunction|SetBuildingTextureData 1 (Utilities|Array|Get(acopy) _color_swatches _swatchindex)))
      (else
        ;[1 1 1] Comment
        (Variables|RandomSwatch|SetSwatchIndex (Math|Integer|SelectInt 0 (- _swatch_count 1) (< (Variables|RandomSwatch|GetSwatchIndex) 0)))))))

(fn ValidateVisualsFlipBook
  (bind (_x _y) (Math|Vector2D|BreakVector2D (Variables|FlipBook|GetFlipBookRowColumnCount)))
  (Utilities|CanExecuteFlipBook
    (:Valid
      (Rendering|Material|SetScalarParameterValueonMaterials (Variables|Debug|GetStaticMeshComponent) (Variables|FlipBook|GetFlipBookIndexParameterName) (Math|Conversions|ToFloat(Integer) (Math|Integer|Clamp(Integer) (Variables|FlipBook|GetFlipBookIndex) 0 (Math|Float|Floor (* _x _y))))))))
